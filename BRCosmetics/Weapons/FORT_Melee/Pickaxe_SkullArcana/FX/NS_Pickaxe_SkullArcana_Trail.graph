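# Niagara swing trail for the Skull Arcana pickaxe, built from ribbons that follow pairs of bones or sockets through shared anim-trail modules.
# AnimTrail_Nebula draws a short-lived, tapering trail animated through dynamic material parameter curves.
# AnimTrail_SharpEdge draws two bright, camera-offset trails, split by spawn group for width and opacity, through two ribbon renderers.
# Time-of-day colors tint both emitters, and curves fade and narrow the ribbons over their lifetime.
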NiagaraSystem /BRCosmetics/Weapons/FORT_Melee/Pickaxe_SkullArcana/FX/NS_Pickaxe_SkullArcana_Trail
dependencies {
 usedDataInterfaces [&NiagaraDataInterfaceCurve, &NiagaraDataInterfaceSocketReader, &NiagaraDataInterfaceColorCurve]
 usedDynamicInputs [&ColorFromCurve]
 usedModules [
  &self:SystemSpawnScript.SystemScriptSource.SystemScriptGraph.NiagaraNodeAssignment_1.SetVariables_#1
  &self:SystemSpawnScript.SystemScriptSource.SystemScriptGraph.NiagaraNodeAssignment_0.SetVariables_#2
  &SystemState
 ]
}
emitters {
 AnimTrail_Nebula {
  emitterUpdateScript [
   &EmitterState()
   &FN_SpawnAnimTrail()
  ]
  particleSpawnScript [
   &FN_InitializeAnimTrailBoneIndecies(First Bone or Socket Index=2, Second Bone or Socket Index=3)
   &FN_InitializeAnimTrail(Color=$System.TODColorMultiplier, Lifetime=0.25)
  ]
  particleUpdateScript [
   &FN_UpdateAnimTrail()
   &DynamicMaterialParameters(
    Index 0 Param 1=&FloatFromCurve(FloatCurve=DataInterface() {propertyValues {Curve {keys [<at 0 0 RCIM_Cubic>,<at 0.4 0.8 RCIM_Cubic>,<at 0.6 0 RCIM_Cubic>]}}})
    Index 0 Param 2=&FloatFromCurve(FloatCurve=DataInterface() {propertyValues {Curve {keys [
         {arriveTangent=0.0872, interpMode=RCIM_Cubic, leaveTangent=0.0872, tangentMode=RCTM_User, time=0.0523, value=1}
         {arriveTangent=-1.3026, interpMode=RCIM_Cubic, leaveTangent=-1.3026, time=0.5, value=0.4}
         {arriveTangent=-1.1763, interpMode=RCIM_Cubic, leaveTangent=-1.1763, tangentMode=RCTM_User, time=0.82, value=0}
        ]}}})
    Index 0 Param 3=1
    Index 0 Param 4=1
    Module.Param0WriteEnabled=true
    Module.Param1WriteEnabled=true
    Module.Param2WriteEnabled=true
    Module.Param3WriteEnabled=true
   )
   &ScaleColor(
    Scale Alpha=&FloatFromCurve(
     FloatCurve=DataInterface() {propertyValues {Curve {keys [
         {arriveTangent=0.3569, interpMode=RCIM_Cubic, leaveTangent=0.3569, tangentMode=RCTM_User, time=0, value=0}
         {arriveTangent=2.3333, interpMode=RCIM_Cubic, leaveTangent=2.3333, time=0.0211, value=0.95}
         {arriveTangent=-1.2627, interpMode=RCIM_Cubic, leaveTangent=-1.2627, time=0.3, value=0.7}
         {arriveTangent=-1, interpMode=RCIM_Cubic, leaveTangent=-1, time=0.6543, value=0.1504}
         <at 1 0 RCIM_Cubic>
        ]}}}
     Scale Curve=0.8
    )
    ScaleA=true
    ScaleRGB=true
   )
   &ScaleRibbonWidth(Ribbon Width Scale=&FloatFromCurve(FloatCurve=DataInterface() {propertyValues {Curve {keys [<at 0 1 RCIM_Cubic>, {arriveTangent=-1.6223, interpMode=RCIM_Cubic, leaveTangent=-1.6223, tangentMode=RCTM_User, time=1, value=0}]}}}))
  ]
  rendererClasses [&NiagaraRibbonRendererProperties]
  renderers [{rendererClass=&NiagaraRibbonRendererProperties}]
 }
 AnimTrail_SharpEdge {
  emitterUpdateScript [
   &EmitterState()
   &FN_SpawnAnimTrail(Number of Anim Trails=2)
  ]
  particleSpawnScript [
   &FN_InitializeAnimTrailBoneIndecies(First Bone or Socket Index=1, Second Bone or Socket Index=0)
   &FN_InitializeAnimTrailBoneIndecies(AnimTrail Selection=1, Second Bone or Socket Index=0) {module=FN_InitializeAnimTrailBoneIndecies001}
   &FN_InitializeAnimTrail(
    AnimTrailWidth=&MergeHelper_FloatSwitch(0=1, 1=7.5, 2=?, 3=?, Index=&SpawnGroup(), Num=::NewEnumerator0)
    Color=<rgba 2.5 1.625 0.625 1>
    Lifetime=0.25
   )
   &CameraOffset(Camera Offset Amount=5)
  ]
  particleUpdateScript [
   &FN_UpdateAnimTrail()
   &ScaleColor(
    Scale Alpha=&FloatFromCurve(
     FloatCurve=DataInterface() {propertyValues {Curve {keys [
         {arriveTangent=4.0681, interpMode=RCIM_Cubic, leaveTangent=4.0681, tangentMode=RCTM_User, time=0, value=0.6824}
         {arriveTangent=1.059, interpMode=RCIM_Cubic, leaveTangent=1.059, time=0.0896, value=0.9735}
         {arriveTangent=-1.8159, interpMode=RCIM_Cubic, leaveTangent=-1.8159, time=0.2055, value=0.9}
         {arriveTangent=-1.1328, interpMode=RCIM_Cubic, leaveTangent=-1.1328, time=0.5221, value=0.1882}
         <at 1 0 RCIM_Cubic>
        ]}}}
     Scale Curve=&MaskFloatBySpawnGroup(Masked Float=0, Passthrough Float=0.1, Spawn Group Mask=1)
    )
    Scale RGB=&MakeVectorFromLinearColorRGB(Color=&LerpLinearColors(StartColor=<rgba 1 1 1 1>, EndColor=$System.TODColorMultiplier, LerpFactor=0.6))
    ScaleA=true
    ScaleRGB=true
   )
   &ScaleRibbonWidth(Ribbon Width Scale=&FloatFromCurve(FloatCurve=DataInterface() {propertyValues {Curve {keys [<at 0 1 RCIM_Cubic>, {arriveTangent=-1.6223, interpMode=RCIM_Cubic, leaveTangent=-1.6223, tangentMode=RCTM_User, time=1, value=0}]}}}))
   /SetVariables(Particles.ColorScreenFacing:LinearColor=&MultiplyLinearColor(
     Linear Color=$Particles.Initial.Color
     Scale Factor=&MakeVector4FromVectorAndFloat(
      "Vector (RGB)"=&MakeVectorFromLinearColorRGB(Color=&FNTimeOfDayColor(Night Color=<rgba 0.32 0.4976 0.8 1>, Morning Color=<rgba 0.6 0.6262 0.8 1>, Day Color=<rgba 1 1 1 1>, Evening Color=<rgba 0.6 0.6 0.6 1>))
      "Float (Alpha)"=&FloatFromCurve(
       FloatCurve=DataInterface() {propertyValues {Curve {keys [{interpMode=RCIM_Cubic, tangentMode=RCTM_User, time=0, value=1}, {interpMode=RCIM_Cubic, tangentMode=RCTM_User, time=1, value=0}]}}}
       Scale Curve=&FloatFromCurve(
        FloatCurve=DataInterface() {propertyValues {Curve {keys [
            {arriveTangent=22.4583, interpMode=RCIM_Cubic, leaveTangent=22.4583, tangentMode=RCTM_User, time=0, value=0}
            {arriveTangent=-0.0759, interpMode=RCIM_Cubic, leaveTangent=-0.0759, tangentMode=RCTM_User, time=0.0285, value=0.9173}
            {arriveTangent=-7.4744, interpMode=RCIM_Cubic, leaveTangent=-7.4744, tangentMode=RCTM_User, time=0.1421, value=0.6102}
            {arriveTangent=2.5339, interpMode=RCIM_Cubic, leaveTangent=2.5339, tangentMode=RCTM_User, time=0.3, value=0}
           ]}}}
        CurveIndex=$Emitter.Age
        Scale Curve=&MaskFloatBySpawnGroup(Masked Float=0.25, Passthrough Float=0, Spawn Group Mask=1)
       )
      )
     )
    )) {isSetParameters=true}
  ]
  rendererClasses [&NiagaraRibbonRendererProperties]
  renderers [
   {rendererClass=&NiagaraRibbonRendererProperties}
   {rendererClass=&NiagaraRibbonRendererProperties, rendererIndex=1}
  ]
 }
}
summary {
 systemName=NS_Pickaxe_SkullArcana_Trail
 emitters [
  {emitterName=AnimTrail_Nebula, rendererClasses [&NiagaraRibbonRendererProperties]}
  {emitterName=AnimTrail_SharpEdge, rendererClasses [&NiagaraRibbonRendererProperties]}
 ]
 userVariables [User.RibbonIDOffset {type=NiagaraInt32}]
}
